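; Secondary-fire Blueprint for HeatSnake Drift that chooses which lightweight projectile each shot fires.
; GetLightWeightProjectile takes the secondary shot counter modulo the truncated ExplodeOnThisBulletCount scalable float.
; When the remainder is zero it returns the parent's projectile, presumably the exploding round going by the variable name.
; All other shots return the Standard secondary projectile class.
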(fn GetLightWeightProjectile
  (if (== (Math|Integer|%25%28Integer%29 (Variables|HeatSnakeDrift|GetSecondaryShotCounter) (Math|Float|Truncate (ScalableFloat|GetValueatLevel (Variables|HeatSnakeDrift|GetExplodeOnThisBulletCount)))) 0)
    (bind _returnvalue (|CallParentFunction))
    (return _returnvalue)
    (else
      (return "/HeatSnake_Drift/Gameplay/Prj_HeatSnake_Drift_Secondary_Standard.Prj_HeatSnake_Drift_Secondary_Standard_C"))))

(fn ConstructionScript)
