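; HeatSnake Drift weapon Blueprint, a generic Athena weapon base that drives the gun's cosmetic effects.
; Fire, reload, equip and visibility events drive muzzle flash, shell ejection, reload smoke, magazine eject, wind particles and force feedback, with exotic variants flagged to the muzzle system.
; Cosmetic alterations, fire-mode data and magazine mods override effect assets.
; It also handles rate-of-fire sound pitch, pierce-through, reticle visibility, durability-break effects and Niagara-callback magazine impact audio.
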(event Custom|BindFireRateChange
  ;[1 1 1] If Auto weapon, Bind Rate of fire change event for fire sound pitch then call it one time to set it initially
  (switch Utilities|FlowControl|Switch|SwitchonEFortWeaponTriggerType (WeaponItemDefinition|GetWeaponDataTriggerType)
    (:Automatic
      (FortWeapon|BindEventtoOnWeaponRateOfFireChanged (delegate Custom|PitchUpOnRateOfFireChange))
      (CallFunction|PitchUponRateOfFireChange (FortWeapon|GetWeaponRateOfFireModifier)))))

(event Custom|ShellEjectionFixOn
  ;[1 1 1] Fix for Shell Ejection Issue on Quality Change
  (EventDispatchers|BindEventtoOnEffectsQualityChanged (Settings|GetFortGameUserSettings) (delegate Custom|Bind%20on%20Effects%20Quality)))

(event Custom|ShellEjectionOff
  ;[1 1 1] Fix for Shell Ejection Issue on Quality Change
  (EventDispatchers|UnbindEventfromOnEffectsQualityChanged (Settings|GetFortGameUserSettings) (delegate Custom|Bind%20on%20Effects%20Quality)))

(event Custom|PitchUpOnRateOfFireChange
  (params (NewRateOfFire :type Float%20%28single-precision%29))
  ;[1 1 1] If Auto weapon, Bind Rate of fire change event for fire sound pitch then call it one time to set it initially
  ;;[1 1 1] Pitch on weapon rate of fire change
  (FortWeapon|SetFireSoundPitchMultiplier (Math|Curves|GetFloatValue (Variables|Default|GetCurve_PitchOffset) NewRateOfFire)))

(event Custom|Bind%20on%20Effects%20Quality
  ;[1 1 1] Fix for Shell Ejection Issue on Quality Change
  (Utilities|FlowControl|Delay 0.4) ; REQUIRED to make work!
  (if (ScalableFloat|GetValueasBool (Variables|Effects|Muzzle|GetUseNativeFX))
    (CallFunction|DeactivateShellsFX)
    (else
      (Utilities|IsValid (Variables|Effects|Shells|GetSpawned_Shells)
        (:Is%20Valid
          (CallFunction|UpdateShellEmittersFX))))))

(event EventBeginPlay
  (WeaponMods|BindEventtoOnModAddedtoWeapon (EventDispatchers|CreateEvent :!event Custom|ModAddedtoWeapon))
  (CallFunction|UpdateModMagazine)
  (Utilities|FlowControl|Sequence
    (:then_0
      ;[0 0 0] Send wind particle reference to the wind manager
      (if (Variables|Default|GetIs%20Wind%20Enabled)
        (Wind|AddWindParticleSystemComponent (Fortnite|GetWindManager) (Variables|Effects|Muzzle|GetMuzzleWindParticleSpawned))))
    (:then_1
      (CallFunction|InitMuzzleFXData))))

(event Custom|ShellsON_(onPump)
  ;[0.3841 0.6674 1] Shells VFX (on Pump)
  (CallFunction|ActivateShellsFX (and (Variables|Effects|Shells|GetUseShellsOnPump?) (> (Rendering|GetRenderingDetailMode) 0))))

(event Custom|ModAddedtoWeapon
  (params (Mod :type Fort%20Weapon%20Mod%20Item%20Definition%20Object%20Reference) (Weapon :type Fort%20Weapon%20Object%20Reference))
  (switch Utilities|FlowControl|Switch|SwitchonGameplayTag (Variables|/Script/FortniteGame.FortWeaponModItemDefinition|GetModSlot Mod)
    (:WeaponMod.Slot.Magazine
      (CallFunction|UpdateModMagazine))))

(event EventOnPlayWeaponFireFX
  (params (bPersistentFire :type Boolean) (bSecondaryFire :type Boolean))
  (bind _spawned_muzzle_system (Variables|Effects|Muzzle|GetSpawned%20Muzzle%20System))
  ;[0.4433 0 1] Muzzle Flash VFX
  (Default|CallOnStartFiring)
  (Utilities|FlowControl|Sequence
    (:then_0
      (CallFunction|MuzzleFlashFX bPersistentFire)
      (bind _instigator (Utilities|GetValidInstigator)
        (:Is%20Valid
          (bind _asfort_player_pawn (Utilities|Casting|CastToFortPlayerPawn _instigator))
          (Rendering|SetFirstPersonPrimitiveType (Variables|Effects|Muzzle|GetSpawned%20Muzzle%20System) (select (Class|FortPlayerPawn|IsHipFireFirstPersonActive _asfort_player_pawn) "FirstPerson" "None"))
          (Effects|Components|ParticleSystem|SetIntParameter _spawned_muzzle_system "Variant" (Variables|Effects|Muzzle|GetVariant))
          (Effects|Components|ParticleSystem|SetFloatParameter _spawned_muzzle_system "IS_Exotic" (Math|Conversions|ToFloat(Boolean) (or (== (Data|GetWeaponData) "/HeatSnake_Drift/Gameplay/WID_HeatSnake_Drift_Secondary_X.WID_HeatSnake_Drift_Secondary_X") (== (Data|GetWeaponData) "/HeatSnake_Drift/Gameplay/WID_HeatSnake_Drift_X.WID_HeatSnake_Drift_X")))))))
    (:then_1
      (CallFunction|ActivateorDeactivateWindParticle true))
    (:then_2
      (if (Effects|ShouldPlayShellFX)
        (CallFunction|ActivateShellsFX (Variables|Effects|Shells|GetUseShellsOnFire?))))
    (:then_3
      (FortWeapon|PlayForceFeedback false bPersistentFire))))

(event Custom|UpdateModMagazine
  (bind _returnvalue (Class|FortWeaponModFunctionLibrary|GetModStaticMeshDatabySlot self "(TagName=\"WeaponMod.Slot.Magazine\")"))
  (Utilities|IsValid (Utilities|ResolveSoftReference (Utilities|Struct|BreakWeaponModStaticMeshBlueprintData _returnvalue))
    (:Is%20Valid
      (Variables|Effects|Reload|SetMagazine%20Eject%20Model (Utilities|ResolveSoftReference (Utilities|Struct|BreakWeaponModStaticMeshBlueprintData _returnvalue)))
      (Variables|Effects|Reload|SetUse%20Magazine%20Eject true))))

(event FortWeapon|EventOnUnEquippedAndDetached
  ;[0.4433 0 1] Muzzle Flash VFX
  (CallFunction|DeactivateMuzzleFX)
  (bind _spawned_muzzle_system (Variables|Effects|Muzzle|GetSpawned%20Muzzle%20System)
    (:then
      (Effects|Components|ParticleSystem|ReleasetoPool _spawned_muzzle_system)
      (Variables|Effects|Muzzle|SetSpawned%20Muzzle%20System 0))
    (:else)))

(event EventEndPlay
  (params (EndPlayReason :type EEndPlayReason%20Enum :default "Destroyed"))
  ;[0.4433 0 1] Muzzle Flash VFX
  (|CallParentFunction EndPlayReason)
  (CallFunction|DeactivateMuzzleFX)
  (bind _spawned_muzzle_system (Variables|Effects|Muzzle|GetSpawned%20Muzzle%20System)
    (:then
      (Effects|Components|ParticleSystem|ReleasetoPool _spawned_muzzle_system)
      (Variables|Effects|Muzzle|SetSpawned%20Muzzle%20System 0))
    (:else)))

(event EventOnStopWeaponFireFX
  ;[0.4433 0 1] Muzzle Flash VFX
  (Default|CallOnPersistentFireStopped)
  ;;[1 1 1] Enforce a short minimum time to play looping FX
  (if (not (Utilities|NewMacro (Utilities|Time|GetGameTimeinSeconds)))
    (Utilities|FlowControl|RetriggerableDelay 0.0))
  (CallFunction|ActivateorDeactivateWindParticle false)
  (CallFunction|DeactivateMuzzleFX)
  (Effects|DeactivateBeamFX)
  (CallFunction|DeactivateShellsFX)
  (FortWeapon|StopForceFeedback))

(event EventOnInitCosmeticAlterations
  (params (CosmeticMod :type Fort%20Cosmetic%20Modification%20Structure))
  (bind (_cosmeticmaterial _ambientparticlesystem _ambientniagarasystem _muzzleparticlesystem _muzzleniagarasystem _reloadparticlesystem _reloadniagarasystem _beamparticlesystem _beamniagarasystem _tracertemplate _bmodifycolor _coloralteration _colorparametername _bmodifydecalcolour _decalcolouralterationstart _decalcolouralterationend _bmodifyshellcolour _shellcolouralteration) (Utilities|Struct|BreakFortCosmeticModification CosmeticMod))
  (bind _returnvalue (CallFunction|GetBeamPSC))
  ;[0.06096 0.8755 1] Setup Alteration FX
  ;; Alteration Logic
  (Networking|CanExecuteCosmeticEvents
    (:True
      (Utilities|FlowControl|Sequence
        (:then_0
          (if (not (Asset|IsAssetNull _ambientniagarasystem))
            (bind _object (Utilities|AsyncLoadAsset _ambientniagarasystem)
              (:Completed
                (bind _asniagara_system (Utilities|Casting|CastToNiagaraSystem _object))
                (bind _alteration_ambient_ps (Variables|Effects|Alterations|GetAlteration%20Ambient%20PS)
                  (:then
                    (Components|DestroyComponent _alteration_ambient_ps))
                  (:else
                    _))
                (bind _niagara_component_system_reference (Utilities|ElementalWeaponSetupFX (Variables|Visual|GetWeaponMesh) _asniagara_system "FX_Idle_Start" "FX_Idle_End" false)
                  (:Out
                    (bind _output_get (Variables|Effects|Alterations|SetAlteration%20Ambient%20PS _niagara_component_system_reference))
                    (LOD|SetMaxDrawDistance _output_get (* (Variables|Effects|DrawDistances|GetAlteration%20Ambient%20PS%20Max%20Draw%20Distance) (Utilities|Reciprocal (Fortnite|GetNumericCVar "r.ViewDistanceScale"))))))))))
        (:then_1
          (if (Asset|IsAssetNull _muzzleniagarasystem)
            (if (not (Asset|IsAssetNull _muzzleparticlesystem))
              (bind _object_2 (Utilities|AsyncLoadAsset _muzzleparticlesystem))
              (bind _asfxsystem_asset (Utilities|Casting|CastToFXSystemAsset _object_2))
              (Variables|Effects|Muzzle|SetMuzzle%20System _asfxsystem_asset))
            (else
              (bind _object_3 (Utilities|AsyncLoadAsset _muzzleniagarasystem))
              (bind _asfxsystem_asset_2 (Utilities|Casting|CastToFXSystemAsset _object_3))
              (Variables|Effects|Muzzle|SetMuzzle%20System _asfxsystem_asset_2))))
        (:then_2
          (bind _returnvalue_2 (Effects|Components|ParticleSystem|SelectAltSystemAsset "WeaponRangedBeam" _beamniagarasystem _beamparticlesystem))
          (if (and (Effects|GetUseBeamParticles) (not (Asset|IsAssetNull _returnvalue_2)))
            (bind _object_4 (Utilities|AsyncLoadAsset _returnvalue_2)
              (:Completed
                (bind _asparticle_system (Utilities|Casting|CastToParticleSystem _object_4)
                  (:then
                    (bind _returnvalue_3 (Effects|SetBeamParticleSystem _asparticle_system))
                    (Effects|Components|ParticleSystem|SetTemplate _returnvalue _returnvalue_3)
                    (LOD|SetMaxDrawDistance _returnvalue (* (Variables|Effects|DrawDistances|GetBeam%20PS%20Max%20Draw%20Distance) (Utilities|Reciprocal (Fortnite|GetNumericCVar "r.ViewDistanceScale")))))
                  (:CastFailed
                    (bind _asniagara_system_2 (Utilities|Casting|CastToNiagaraSystem _object_4))
                    (Effects|SetBeamNiagaraSystem _asniagara_system_2)))))))
        (:then_3
          (if _bmodifydecalcolour
            (Decals|SetDecalColorStart _decalcolouralterationstart)
            (Decals|SetDecalColorEnd _decalcolouralterationend)))
        (:then_4
          (if _bmodifyshellcolour
            (Variables|Effects|Shells|SetShell%20Color _shellcolouralteration)))
        (:then_5
          (bind _asplayer_pawn_generic_parent (Utilities|Casting|CastToPlayerPawn_Generic_Parent (Game|GetInstigator self)))
          (Class|PlayerPawnGenericParent|ReinitializeWeaponMaterials _asplayer_pawn_generic_parent))))))

(event EventOnPlayReloadFX
  (params (ReloadStage :type EFortReloadFXState%20Enum :default "ReloadStart"))
  ;[0.3518 0.0375 0.63] Use Reload Smoke Trails
  (Default|MuzzlePlayReloadFX ReloadStage)
  (FortWeapon|PlayForceFeedback true))

(event EventOnInitAlteration
  (params (NewAlteration :type Fort%20Alteration%20Item%20Definition%20Object%20Reference)))

(event FortWeapon|EventOnEquippedAndAttached
  ;[1 1 1] On weapon equip
  (bind _asfort_player_pawn (Utilities|Casting|CastToFortPlayerPawn (Game|GetInstigator))
    (:then
      (Variables|Default|SetbIsLocal (Pawn|IsLocallyViewed _asfort_player_pawn)))
    (:CastFailed
      _))
  (EventDispatchers|BindEventtoOnLocalReloadStarted (EventDispatchers|CreateEvent :!event LocalReloadStarted))
  ;;[1 1 1] Check if in athena, if not set up fire rate bind.
  (if (not (FortGlobals|IsinAthena))
    (Networking|SwitchHasAuthority
      (:Authority
        _)
      (:Remote
        (CallFunction|BindFireRateChange))))
  (CallFunction|TryShowReticle)
  (CallFunction|ShellEjectionFixOn)
  (Utilities|FlowControl|Delay 0.1) ; This is a hack fix
  (CallFunction|SetActiveAlterationIdleParticles true))

(event FortWeapon|EventOnLocalUnEquip
  (CallFunction|DeactivateMuzzleFX)
  (CallFunction|SetActiveAlterationIdleParticles false)
  (CallFunction|DeactivateShellsFXNative "None")
  (bind _instigator (Utilities|GetValidInstigator)
    (:Is%20Valid
      (bind _asfort_pawn (Utilities|Casting|CastToFortPawn _instigator))
      (Utilities|IsValid (Pawn|GetLocalViewingPlayerController _asfort_pawn)
        (:Is%20Valid
          (CallFunction|TryShowReticle))))))

(event EventOnWeaponVisibilityChanged
  (params (bVisible :type Boolean) (bSetForLocalControllerOnly :type Boolean))
  (bind _alteration_ambient_ps (Variables|Effects|Alterations|GetAlteration%20Ambient%20PS))
  ;[1 1 1] EVENT -> OnWeaponVisibilityChanged
  (Utilities|IsValid (Variables|Effects|Alterations|GetAlteration%20Ambient%20PS)
    (:Is%20Valid
      (if bVisible
        (Components|Activation|Activate _alteration_ambient_ps false)
        (else
          (Components|Activation|Deactivate _alteration_ambient_ps)
          (CallFunction|DeactivateShellsFXNative "None")
          (bind _spawned_shells (Variables|Effects|Shells|GetSpawned_Shells)
            (:then
              (Effects|Components|ParticleSystem|ReleasetoPool _spawned_shells)
              (Variables|Effects|Shells|SetSpawned_Shells 0))
            (:else)))))
    (:Is%20Not%20Valid
      (if bVisible
        _)))
  (Rendering|SetVisibility (Variables|Visual|GetWeaponMesh) bVisible false))

(event EventOnApplyFireModeData
  (params (FireModeData :type Fort%20Weapon%20Fire%20Mode%20Data%20Object%20Reference))
  (bind _muzzleniagarasystem (Variables|/Script/FortniteGame.FortWeaponFireModeData|GetMuzzleNiagaraSystem FireModeData))
  (bind _muzzleparticlesystem (Variables|/Script/FortniteGame.FortWeaponFireModeData|GetMuzzleParticleSystem FireModeData))
  ;[1 1 1] override muzzle FX from alternative fire mode data.
  (Networking|CanExecuteCosmeticEvents
    (:True
      (CallFunction|DeactivateMuzzleFX) ; stop playing muzzle flash as \nsetting to new particle template \nresets fx.
      (if (Utilities|IsValid _muzzleniagarasystem :!class /Script/Engine.KismetSystemLibrary)
        (Variables|Effects|Muzzle|SetMuzzle%20System _muzzleniagarasystem)
        (Variables|Effects|Muzzle|SetSpawned%20Muzzle%20System 0)
        (elif (Utilities|IsValid _muzzleparticlesystem :!class /Script/Engine.KismetSystemLibrary)
          (Variables|Effects|Muzzle|SetMuzzle%20System _muzzleparticlesystem)
          (Variables|Effects|Muzzle|SetSpawned%20Muzzle%20System 0))))))

(event EventOnEquippedWeaponDestory
  (bind (_tintcolor _drawas _tiling _mirroring _imagesize _margin _resourceobject) (Utilities|Struct|BreakSlateBrush (Item|GetWeaponDataSmallPreviewImageBrush)))
  ;[0.725 0.6854 0.4277] Show fx for weapon destruction on depleted durability
  (if (Variables|Effects|GetUseDestroyEffect)
    (if (FortGlobals|IsinAthena)
      (Effects|SpawnEmitteratLocation (Variables|Effects|GetWeaponDurabilityDestroyEffect) (Transformation|GetSocketLocation (Variables|Visual|GetWeaponMesh) "Handle"))
      (bind _instigator (Utilities|GetValidInstigator)
        (:Is%20Valid
          (bind _asfort_pawn (Utilities|Casting|CastToFortPawn _instigator))
          (bind _returnvalue (Effects|SpawnEmitterAttached :EmitterTemplate (Variables|Effects|GetWeaponDurabilityDestroyEffectIcon) :AttachToComponent (Variables|/Script/Engine.Character|GetMesh _asfort_pawn) :AttachPointName "Root_Effect" :Location "0, 0,130.000000" :Scale "1.000000,1.000000,1.000000" :LocationType "SnapToTarget" :bAutoDestroy true :PoolingMethod "None" :bAutoActivate true))
          (bind _returnvalue_2 (Rendering|Material|CreateDynamicMaterialInstance _returnvalue 0 "/Game/Effects/WeaponDestroy/M_WeaponDestroyExclamation.M_WeaponDestroyExclamation" "None" :!class /Script/Engine.PrimitiveComponent))
          (Rendering|Material|SetMaterial _returnvalue 0 _returnvalue_2)
          (bind _astexture_2d (Utilities|Casting|CastToTexture2D _resourceobject)
            (:then
              _)
            (:CastFailed
              _))
          (Rendering|Material|SetTextureParameterValue _returnvalue_2 "IconTex" _astexture_2d)))
      (else
        (GameplayCue|ExecuteGameplayCue%28Burst%29Local self (Variables|Effects|GetWeapon%20Durability%20Destroy%20Gameplay%20Cue) (Ability|GameplayCue|MakeGameplayCueParameters 0.0))))))

(event Niagara|EventReceiveParticleData
  (params (Data :type Array%20of%20Basic%20Particle%20Data%20Structures) (NiagaraSystem :type Niagara%20System%20Object%20Reference) (SimulationPositionOffset :type Vector%20%28by%20ref%29 :default "0, 0, 0"))
  ;[0.7404 0.7657 1] Plays magazine impact audio from magazine NS collision callback if UseMagazineEject FX is true . Local player only, see ActivateMagazineEjectFX for setup
  (Networking|CanExecuteCosmeticEvents
    (:True
      (if (Variables|Default|GetbIsLocal)
        (if (> (Utilities|Array|Length Data) 0)
          (for _array_element Data
            (bind _returnvalue (SoundLibrary|PlaySoundSimple (WorldSubsystems|GetSoundLibrarySubsystem) self (Utilities|Struct|MakeSimpleSoundSettings "(TagName=\"SoundLibrary.Weapon.Reload.Cartridge.Impact\")" false false true true 0.0 1.0 0.0 0 0 "None" 0 "KeepWorldPosition") :!class /Script/SoundLibrary.SoundLibrarySubsystem))
            (if (> (Utilities|Array|Length (.AudioComponents (Utilities|Struct|BreakSoundLibraryPlayResult _returnvalue))) 0)
              (for _array_element_2 (.AudioComponents (Utilities|Struct|BreakSoundLibraryPlayResult _returnvalue))
                (Transformation|SetWorldLocation _array_element_2 (Utilities|Struct|BreakBasicParticleData _array_element) false false)))))))))

(event Custom|SetWeaponPierceThrough
  (params (Enable :type Boolean) (TargetLimit :type Integer))
  (bind _returnvalue (Math|Integer|SelectInt TargetLimit 0 Enable))
  ;[1 1 1] Event to dynamically enable/disable trace through pawns
  (FortWeapon|SetTraceThroughPawnsLimit _returnvalue)
  (FortWeapon|SetTraceThroughPawns Enable)
  (CallFunction|SetWeaponPierceThroughClientRep Enable _returnvalue))

(event Custom|SetWeaponPierceThrough_ClientRep
  (params (Enable :type Boolean) (TargetLimit :type Integer))
  ;[1 1 1] Event to dynamically enable/disable trace through pawns
  (FortWeapon|SetTraceThroughPawnsLimit (Math|Integer|SelectInt TargetLimit 0 Enable))
  (FortWeapon|SetTraceThroughPawns Enable))

(event Custom|HideWeaponMesh
  ;[1 1 1] DEBUG EVENTS -> ke * <event name>
  (Development|SetHiddeninGame (Variables|Visual|GetWeaponMesh) true false)
  (Development|PrintString "HideWeaponMesh"))

(event Custom|ShowWeaponMesh
  ;[1 1 1] DEBUG EVENTS -> ke * <event name>
  (Development|SetHiddeninGame (Variables|Visual|GetWeaponMesh) false false)
  (Development|PrintString "ShowWeaponMesh"))

(event Custom|HideWeapon
  ;[1 1 1] DEBUG EVENTS -> ke * <event name>
  (FortWeapon|SetWeaponVisibility false true)
  (Development|PrintString "HideWeapon"))

(event FortWeapon|EventOnPlayImpactFX
  (params (HitResult :type Hit%20Result%20Structure%20%28by%20ref%29) (ImpactPhysicalSurface :type EPhysicalSurface%20Enum :default "SurfaceType_Default") (SpawnedPSC :type FXSystem%20Component%20Object%20Reference))
  ;[1 1 1] Use native implementation
  (|CallParentFunction HitResult ImpactPhysicalSurface SpawnedPSC)
  (Effects|Components|ParticleSystem|SetIntParameter SpawnedPSC "WeaponType" (Math|Conversions|ToInteger(Byte) (FortWeapon|GetCoreAnimation)))
  (Effects|Components|ParticleSystem|SetIntParameter SpawnedPSC "PhysicalSurface" (Math|Conversions|ToInteger(Byte) ImpactPhysicalSurface)))

(event Custom|ShowWeapon
  ;[1 1 1] DEBUG EVENTS -> ke * <event name>
  (FortWeapon|SetWeaponVisibility true true)
  (Development|PrintString "ShowWeapon"))

(fn ActivateMagazineEjectFX
  (Utilities|IsValid (Variables|Effects|Reload|GetMagazine%20Eject%20System)
    (:Is%20Valid
      (bind _returnvalue (Niagara|SpawnSystemAttached (Variables|Effects|Reload|GetMagazine%20Eject%20System) (Variables|Visual|GetWeaponMesh) (Variables|Effects|Reload|GetMagazine%20Eject%20Socket%20Name) "0, 0, 0" "0, 0, 0" "SnapToTarget" true))
      (Niagara|SetNiagaraVariable(StaticMesh) _returnvalue "StaticMesh" (Variables|Effects|Reload|GetMagazine%20Eject%20Model))
      ;[0.1003 0.5009 1] Audio callback for magazine impacts. See EventReceiveParticleData for implementation
      (Niagara|SetNiagaraVariableByString(Object) _returnvalue "User.BPAudioCallback" self))))

(fn ActivateOrDeactivateWindParticle
  (params (bNewActive :type Boolean))
  (if (Variables|Default|GetIs%20Wind%20Enabled)
    (Components|Activation|SetActive (Variables|Effects|Muzzle|GetMuzzleWindParticleSpawned) bNewActive true))
  (return))

(fn ActivateReloadSmokeFX
  (if (Effects|ShouldPlayReloadFX)
    (bind _returnvalue (Niagara|SpawnSystemAttached (Variables|Effects|Reload|GetReload_System) (Variables|Visual|GetWeaponMesh) (Variables|Effects|Reload|GetReloadSocketName) "0, 0, 0" "0, 0, 0" "SnapToTarget" true false))
    (Niagara|SetNiagaraVariable(Bool) _returnvalue "UseEmissiveSmoke" (Variables|Effects|Reload|GetUse%20Emissive%20in%20Reload))
    (Niagara|SetNiagaraVariable(LinearColor) _returnvalue "EmissiveColor" (Variables|Effects|Reload|GetReload%20Emissive%20Color))
    (Niagara|SetNiagaraVariable(Float) _returnvalue "RibbonLifetime" (Variables|Effects|Reload|GetReload%20Smoke%20Lifetime))
    (Niagara|SetNiagaraVariable(Float) _returnvalue "RibbonScale" (Variables|Effects|Reload|GetReload%20Smoke%20Width%20Scale))
    (Niagara|SetNiagaraVariable(Material) _returnvalue "RibbonMaterial" (Variables|Effects|Reload|GetReload%20Smoke%20Material))
    (Niagara|SetNiagaraVariable(Bool) _returnvalue "bIsRocketLauncher" (Variables|Effects|Reload|GetIs%20Rocket%20Launcher))
    (Niagara|SetNiagaraVariable(Bool) _returnvalue "HasShellEjects" (Variables|Effects|Reload|GetReload%20Ejects%20Shells))
    (Niagara|SetNiagaraVariable(LinearColor) _returnvalue "BaseColorAlpha" (Variables|Effects|Reload|GetBaseColorAlpha))
    (Niagara|SetNiagaraVariable(LinearColor) _returnvalue "SparksColor" (Variables|Effects|Reload|GetRocket%20Launcher%20Spark%20Color))
    (Transformation|SetWorldLocation _returnvalue (Effects|GetMuzzleLocation -1) false false)
    (Components|Activation|Activate _returnvalue false)))

(fn ActivateShellsFX
  (params (Bool :type Boolean))
  (if Bool
    (if (ScalableFloat|GetValueasBool (Variables|Effects|Muzzle|GetUseNativeFX))
      (CallFunction|ActivateShellsFXNative (Variables|Effects|Shells|GetShell%20Burst%20FX) (Variables|Effects|Shells|GetShell%20Looping%20FX) (Variables|Effects|Shells|GetShells%20Socket%20Name) "ShellsRotationRate" (Variables|Effects|Shells|GetShellsRotationRate) "Velocity" (Variables|Effects|Shells|GetShells%20Velocity) "Gravity" (Variables|Effects|Shells|GetShells%20Gravity) "Size" (Variables|Effects|Shells|GetShells%20Size) "InheritParentVelocity" (Math|Conversions|ToVector(Float) (Variables|Effects|Shells|GetInherit%20Parent%20Velocity)) "CylinderRadius" (Variables|Effects|Shells|GetCylindrical%20Radius) "CylinderHeight" (Variables|Effects|Shells|GetCylindrical%20Height) "ShellColor" (Variables|Effects|Shells|GetShell%20Color) "Shell Type" (Utilities|Select :Index (Variables|Effects|Shells|GetShellTypeSelect) :NewEnumerator0 0 :NewEnumerator1 1 :NewEnumerator2 2 :NewEnumerator3 4 :NewEnumerator4 3) (Variables|Effects|DrawDistances|GetShells%20PS%20Max%20Draw%20Distance) (Variables|Effects|Shells|GetUseShellsOnReload?) "SpawnProbability" (Variables|Effects|Shells|GetShells%20Spawn%20Rate%20Scale) "IsSystemLocal" (Variables|Default|GetbIsLocal) "UseShellSmoke" (Variables|Effects|Shells|GetbUseShellSmoke))
      (else
        (Utilities|IsValid (Variables|Effects|Shells|GetSpawned_Shells)
          (:Is%20Valid
            (Development|PrintString "Shells Already Valid - Activating" false false "(R=0.000000,G=0.660000,B=1.000000,A=1.000000)" 3.0))
          (:Is%20Not%20Valid
            (CallFunction|SetupShellFX)))
        (Components|Activation|Activate (Variables|Effects|Shells|GetSpawned_Shells) true)
        (CallFunction|UpdateShellEmittersFX)))))

(fn DeactivateMuzzleFX
  (Components|Activation|Deactivate (Variables|Effects|Muzzle|GetSpawned%20Muzzle%20System)))

(fn DeactivateShellsFX
  (bind _spawned_shells (Variables|Effects|Shells|GetSpawned_Shells))
  (if (ScalableFloat|GetValueasBool (Variables|Effects|Muzzle|GetUseNativeFX))
    (CallFunction|DeactivateShellsFXNative "SpawnProbability")
    (elif (Components|Activation|IsActive (Variables|Effects|Shells|GetSpawned_Shells))
      (Niagara|SetNiagaraVariable(Float) _spawned_shells "SpawnProbability" 0.0)))
  (Components|Activation|Deactivate _spawned_shells)
  (Effects|Components|ParticleSystem|ReleasetoPool _spawned_shells)
  (Variables|Effects|Shells|SetSpawned_Shells 0))

(fn HideReticle
  (bind _asfort_player_controller_zone (Utilities|Casting|CastToFortPlayerControllerZone (Game|GetInstigatorController)))
  (HUD|SetHUDElementVisibilitywithContext _asfort_player_controller_zone (Variables|Scope|GetReticleHUDElementTags) true "Reticle" ""))

(fn InitMuzzleFXData
  (Utilities|Struct|SetmembersinFortWeaponMuzzleFXData (Variables|Effects|GetMuzzleFXData) (Variables|Effects|Muzzle|GetMuzzle%20System) (Variables|Effects|Muzzle|GetMuzzleSmokeInitialColor) (Variables|Effects|Muzzle|GetMuzzle%20Core%20Scale) (Variables|Effects|Muzzle|GetMuzzle%20Core%20Scale_FPS) (Variables|Effects|Muzzle|GetDecal%20Offset) (Variables|Effects|Muzzle|GetMuzzle%20Hue%20Control) (Variables|Effects|Muzzle|GetMuzzle%20Core%20Cap%20Scale) (Variables|Effects|Muzzle|GetMuzzle%20Core%20Cap%20Scale_FPS) (Variables|Effects|Muzzle|GetSpark%20Amount%20Scale) (Variables|Effects|Muzzle|GetSmoke%20Initial%20Color%20Scale%20Variance) (Variables|Effects|Muzzle|GetSmoke%20Emissive%20Scale) (Variables|Effects|Muzzle|GetSmoke%20Size%20Scale) (Variables|Effects|Muzzle|GetSmoke%20Size%20Scale_FPS) (Variables|Effects|Muzzle|GetDecal%20Size) (Variables|Effects|Muzzle|GetDecal%20Alpha) (Variables|Effects|DrawDistances|GetMuzzle%20PS%20Max%20Draw%20Distance) (Variables|Effects|Muzzle|GetBurst%20Loops) (Math|MakeBitmask 2) (Math|MakeBitmask 4) (Math|MakeBitmask 8) (Math|MakeBitmask 16) (Math|MakeBitmask 1) (Variables|Effects|Muzzle|GetMuzzle_BakedSupressor%20%28Always%20Suppressed%29) (Variables|Effects|Muzzle|GetUse%20Heated%20Elements) (Variables|Default|GetBurst%20Activated%20Muzzle) (Variables|Effects|Muzzle|GetBaked%20Optics%20%28Always%20has%20Optics%20element%29)))

(fn LocalReloadStarted
  (params (ReloadTime :type Float%20%28single-precision%29) (ReloadType :type EFortWeaponReloadType%20Enum :default "ReloadWholeClip")))

(fn MuzzleFlashFX
  (params (Persistent%20Fire :type Boolean))
  (Effects|NativeMuzzleFlashFX (Variables|GetPersistent%20Fire) (Utilities|Select :Index (Variables|Effects|Shells|GetShellTypeSelect) :NewEnumerator0 0 :NewEnumerator1 1 :NewEnumerator2 2 :NewEnumerator3 3 :NewEnumerator4 4) (Variables|Effects|Muzzle|GetSpawned%20Muzzle%20System) (Variables|Default|GetMuzzleInt) (Variables|Effects|Muzzle|GetLastPlayFXTime) (Variables|Default|GetMuzzle%20FPS%20Active))
  (return))

(fn MuzzlePlayReloadFX
  (params (Selection :type EFortReloadFXState%20Enum :default "ReloadStart"))
  (switch Utilities|FlowControl|Switch|SwitchonEFortReloadFXState Selection
    (:ReloadStart
      ;[1 1 1] Start Reload Smoke
      (if (Variables|Effects|Reload|GetUse%20Reload%20Particles)
        (CallFunction|ActivateReloadSmokeFX)))
    (:ReloadCartridge
      (if (Variables|Effects|Reload|GetUse%20Magazine%20Eject)
        (CallFunction|ActivateMagazineEjectFX)))
    (:Max_None
      ;[1 1 1] Reload Shells
      (CallFunction|ActivateShellsFX (Variables|Effects|Shells|GetUseShellsOnReload?)))))

(fn MuzzleADSCheck
  (bind _bisads (Effects|NativeMuzzleADSCheck (Variables|Default|GetMuzzle%20FPS%20Active)))
  (return _bisads))

(fn MuzzleModCheck
  (bind (_bmuzzlebreak _bmuzzlesupressor _bmuzzlebreacher) (Effects|NativeMuzzleModCheck))
  (return _bmuzzlebreak _bmuzzlesupressor _bmuzzlebreacher))

(fn PlayLocalForceFeedback
  (params (ForceFeedbackEffect :type Force%20Feedback%20Effect%20Object%20Reference) (Tag :type Name) (bLooping :type Boolean))
  (bind _returnvalue (Game|GetInstigatorController))
  (Utilities|IsValid ForceFeedbackEffect
    (:Is%20Valid
      (bind _asplayer_controller (Utilities|Casting|CastToPlayerController _returnvalue))
      (if (Pawn|IsLocalController _returnvalue)
        (Game|Feedback|ClientPlayForceFeedback _asplayer_controller ForceFeedbackEffect Tag bLooping false false)))))

(fn SetActiveAlterationIdleParticles
  (params (Active :type Boolean))
  (Networking|CanExecuteCosmeticEvents
    (:True
      (bind _alteration_ambient_ps (Variables|Effects|Alterations|GetAlteration%20Ambient%20PS)
        (:then
          (if Active
            (Development|SetHiddeninGame _alteration_ambient_ps false false)
            (Components|Activation|Activate _alteration_ambient_ps false)
            (else
              (Development|SetHiddeninGame _alteration_ambient_ps true false)
              (Components|Activation|Deactivate _alteration_ambient_ps)))
          (return))
        (:else)))))

(fn SetPostProcessParams
  (params (InputPin :type Float%20%28double-precision%29))
  (Variables|Scope|SetScopePostProcessBlendWeight 1.0))

(fn SetupShellFX
  (if (Variables|Effects|Shells|GetDebugShellsSocket?)
    (if (Transformation|DoesSocketExist (Variables|Visual|GetWeaponMesh) (Variables|Effects|Shells|GetShells%20Socket%20Name))
      (Development|PrintString "***Success! Socket Found!***" true true "(R=0.000000,G=1.000000,B=0.000000,A=1.000000)" 5)
      (else
        (Development|PrintString "***Error! Cannot find socket name to attach shell eject FX to!***" true true "(R=1.000000,G=0.000000,B=0.000000,A=1.000000)" 5))))
  (bind _returnvalue (Niagara|SpawnSystemAttached (select (Effects|ShouldPlayPersistentFireFX) "/Game/Weapons/FORT_Generic_Effects/ShellEject/NS_ShellsFire_Looping.NS_ShellsFire_Looping" "/Game/Weapons/FORT_Generic_Effects/ShellEject/NS_ShellsFire_Burst.NS_ShellsFire_Burst") (Variables|Visual|GetWeaponMesh) (Variables|Effects|Shells|GetShells%20Socket%20Name) "0, 0, 0" "0, 0, 0" "KeepRelativeOffset" false false "ManualRelease"))
  (bind _output_get (Variables|Effects|Shells|SetSpawned_Shells _returnvalue))
  (Niagara|SetNiagaraVariable(Vector3) _output_get "ShellsRotationRate" (Variables|Effects|Shells|GetShellsRotationRate))
  (Niagara|SetNiagaraVariable(Vector3) _output_get "Velocity" (Variables|Effects|Shells|GetShells%20Velocity))
  (Niagara|SetNiagaraVariable(Vector3) _output_get "Gravity" (Variables|Effects|Shells|GetShells%20Gravity))
  (Niagara|SetNiagaraVariable(Vector3) _output_get "Size" (Variables|Effects|Shells|GetShells%20Size))
  (Niagara|SetNiagaraVariable(Vector3) _output_get "InheritParentVelocity" (Math|Conversions|ToVector(Float) (Variables|Effects|Shells|GetInherit%20Parent%20Velocity)))
  (Niagara|SetNiagaraVariable(Float) _output_get "CylinderRadius" (Variables|Effects|Shells|GetCylindrical%20Radius))
  (Niagara|SetNiagaraVariable(Float) _output_get "CylinderHeight" (Variables|Effects|Shells|GetCylindrical%20Height))
  (Niagara|SetNiagaraVariable(LinearColor) _output_get "ShellColor" (Variables|Effects|Shells|GetShell%20Color))
  (Niagara|SetNiagaraVariable(Int32) _output_get "Shell Type" (Utilities|Select :Index (Variables|Effects|Shells|GetShellTypeSelect) :NewEnumerator0 0 :NewEnumerator1 1 :NewEnumerator2 2 :NewEnumerator3 4 :NewEnumerator4 3))
  (LOD|SetMaxDrawDistance _output_get (* (Variables|Effects|DrawDistances|GetShells%20PS%20Max%20Draw%20Distance) (Utilities|Reciprocal (Fortnite|GetNumericCVar "r.ViewDistanceScale")))) ; FOR EMERGENCIES - Niagara System FX Type should Cull LONG before this setting
)

(fn ShowReticle
  (bind _asfort_player_controller_zone (Utilities|Casting|CastToFortPlayerControllerZone (Game|GetInstigatorController)))
  (HUD|SetHUDElementVisibilitywithContext _asfort_player_controller_zone (Variables|Scope|GetReticleHUDElementTags) false "Reticle" ""))

(fn StopLocalForceFeedback
  (params (ForceFeedbackEffect :type Force%20Feedback%20Effect%20Object%20Reference) (Tag :type Name))
  (bind _returnvalue (Game|GetInstigatorController))
  (Utilities|IsValid ForceFeedbackEffect
    (:Is%20Valid
      (bind _asplayer_controller (Utilities|Casting|CastToPlayerController _returnvalue))
      (if (Pawn|IsLocalController _returnvalue)
        (Game|Feedback|ClientStopForceFeedback _asplayer_controller ForceFeedbackEffect Tag)))))

(fn TryShowReticle
  (Utilities|Casting|CastToFortPlayerControllerGameplay (Game|GetPlayerController 0)
    (:then
      _)
    (:CastFailed
      (if (not (or (Scope|CanUseScopedCameraViaOptic) (Scope|UseScopeTargeting)))
        (CallFunction|HideReticle))))
  (CallFunction|ShowReticle))

(fn UpdateShellEmittersFX
  (Utilities|FlowControl|Sequence
    (:then_0
      (bind _spawned_shells (Variables|Effects|Shells|GetSpawned_Shells)
        (:then
          (Niagara|SetNiagaraVariable(Float) _spawned_shells "SpawnProbability" (select (Variables|Effects|Shells|GetUseShellsOnReload?) (Math|Conversions|ToFloat(Integer) (Utilities|ClampMinInteger (CallFunction|GetShellReloadCounter))) (Variables|Effects|Shells|GetShells%20Spawn%20Rate%20Scale))))
        (:else
          (Development|PrintString "Spawn Shells NOT VALID on Update"))))
    (:then_1
      (if (Variables|Effects|Shells|GetUseShellsOnReload?)
        (CallFunction|SetShellReloadCounter -1))
      (return))))
